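# Post-copy animation Blueprint for the M_MED_SliceVine_LoopPlum player body cosmetic, adding secondary motion to a bag zipper and left pant strap.
# The parent mesh pose feeds a Control Rig, which receives an emote-or-grind float updated every animation tick.
# Weighted multi-bone constraints anchor each dangle's base before world-space AnimDynamics simulates the zipper and three-bone strap chain.
# On customization load, it hides any tail mesh when long backbling is worn.
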AnimBlueprint /BRCosmetics/Characters/Player/Male/Medium/Bodies/M_MED_SliceVine_LoopPlum/Meshes/M_MED_SliceVine_LoopPlum_AnimBP
target_skeleton=&/M_MED_SliceVine_LoopPlum_Skeleton
graphs [
 AnimGraph {nodes [
   n1 = Control Rig!(Source=Copy Pose From Mesh() {bUseAttachedParent=true}, Alpha, EmoteOrGrind:double=Get emoteorgrind() -> (emoteorgrind:double)) -> (Pose) {
    alphaBoolBlend {blendOption=Linear}
    controlRigAssetReference {blueprintRigClass=&/M_MED_SliceVine_LoopPlum_CtrlRig.M_MED_SliceVine_LoopPlum_CtrlRig_C}
   }
   n2 = "Constraint - dyn_bagZipper_base (PPPPPPP)"!(
    ComponentPose=Local To Component(LocalPose=@n1)
    Alpha=1
    ConstraintWeights_0=0.0005
    ConstraintWeights_1:float=0.0097
    ConstraintWeights_2:float=0.0731
    ConstraintWeights_3:float=0.2467
    ConstraintWeights_4:float=0.6633
    ConstraintWeights_5:float=0.0051
    ConstraintWeights_6:float=0.0015
   ) -> (Pose) {
    alphaBoolBlend {blendOption=Linear}
    boneToModify {boneName=dyn_bagZipper_base}
    constraintSetup [
     {targetBone {boneName=pelvis}}
     {targetBone {boneName=spine_01}}
     {targetBone {boneName=spine_02}}
     {targetBone {boneName=spine_03}}
     {targetBone {boneName=spine_04}}
     {targetBone {boneName=spine_05}}
     {targetBone {boneName=clavicle_l}}
    ]
    constraintWeights [1,1,1,1,1,1,1]
   }
   n3 = Anim Dynamics - Bone: dyn_bagZipper(ComponentPose=@n2, Alpha=1, ExternalForce="0,0,0") {
    bAngularSpring=true
    bGravityOverrideInSimSpace=true
    bOverrideAngularDamping=true
    bUseGravityOverride=true
    simulationSpace=World
    alphaBoolBlend {blendOption=Linear}
    boundBone {boneName=dyn_bagZipper}
    componentAppliedLinearAccClamp=<1000 1000 1000>
    componentLinearAccScale=<1 1 1>
    componentLinearVelScale=<1 1 1>
    gravityOverride {z=-980}
    physicsBodyDefinitions [{
      boundBone {boneName=dyn_bagZipper}
      boxExtents=<10 10 10>
      constraintSetup {angularLimitsMax=<0 0 5>, angularLimitsMin=<0 -20 -5>, angularTarget=<0 0 0>}
      localJointOffset=<0.1 0 0>
     }]
    relativeSpaceBone {boneName=root}
   }
   n4 = "Constraint - dyn_pantStrap_l_base (PPPPPP)"!(
    ComponentPose=@n3
    Alpha=1
    ConstraintWeights_0=0.0257
    ConstraintWeights_1:float=0.3532
    ConstraintWeights_2:float=0.4537
    ConstraintWeights_3:float=0.0074
    ConstraintWeights_4:float=0.1593
    ConstraintWeights_5:float=0.0007
   ) -> (Pose) {
    alphaBoolBlend {blendOption=Linear}
    boneToModify {boneName=dyn_pantStrap_l_base}
    constraintSetup [
     {targetBone {boneName=root}}
     {targetBone {boneName=thigh_l}}
     {targetBone {boneName=calf_l}}
     {targetBone {boneName=calf_twist_02_l}}
     {targetBone {boneName=thigh_twist_01_l}}
     {targetBone {boneName=thigh_r}}
    ]
    constraintWeights [1,1,1,1,1,1]
   }
   n5 = Anim Dynamics - Bone: dyn_pantStrap_l_01(ComponentPose=@n4, Alpha=1, ExternalForce="0,0,0") {
    angularDampingOverride=0.9999
    angularSpringConstant=5000
    bGravityOverrideInSimSpace=true
    bUseGravityOverride=true
    simulationSpace=World
    alphaBoolBlend {blendOption=Linear}
    boundBone {boneName=dyn_pantStrap_l_01}
    componentAppliedLinearAccClamp=<1000 1000 1000>
    componentLinearAccScale=<1 1 1>
    componentLinearVelScale=<1 1 1>
    gravityOverride {z=-1000}
    physicsBodyDefinitions [{
      boundBone {boneName=dyn_pantStrap_l_01}
      boxExtents=<5 5 5>
      constraintSetup {angularLimitsMax=<0 2 5>, angularLimitsMin=<0 -60 -5>, angularTarget=<0 0 0>}
      localJointOffset=<0 -1 1>
     }]
    relativeSpaceBone {boneName=root}
   }
   n6 = Anim Dynamics - Bone: dyn_pantStrap_l_02(ComponentPose=@n5, Alpha=1, ExternalForce="0,0,0") {
    angularDampingOverride=0.9999
    angularSpringConstant=5000
    bGravityOverrideInSimSpace=true
    bUseGravityOverride=true
    simulationSpace=World
    alphaBoolBlend {blendOption=Linear}
    boundBone {boneName=dyn_pantStrap_l_02}
    componentAppliedLinearAccClamp=<1000 1000 1000>
    componentLinearAccScale=<1 1 1>
    componentLinearVelScale=<1 1 1>
    gravityOverride {z=-1000}
    physicsBodyDefinitions [{
      boundBone {boneName=dyn_pantStrap_l_02}
      boxExtents=<5 5 5>
      constraintSetup {angularLimitsMax=<0 2 5>, angularLimitsMin=<0 -60 -5>, angularTarget=<0 0 0>}
      localJointOffset=<0 -1 1>
     }]
    relativeSpaceBone {boneName=root}
   }
   n7 = Anim Dynamics - Bone: dyn_pantStrap_l_03(ComponentPose=@n6, Alpha=1, ExternalForce="0,0,0") {
    angularDampingOverride=0.9999
    angularSpringConstant=5000
    bGravityOverrideInSimSpace=true
    bUseGravityOverride=true
    simulationSpace=World
    alphaBoolBlend {blendOption=Linear}
    boundBone {boneName=dyn_pantStrap_l_03}
    componentAppliedLinearAccClamp=<1000 1000 1000>
    componentLinearAccScale=<1 1 1>
    componentLinearVelScale=<1 1 1>
    gravityOverride {z=-1000}
    physicsBodyDefinitions [{
      boundBone {boneName=dyn_pantStrap_l_03}
      boxExtents=<5 5 5>
      constraintSetup {angularLimitsMax=<0 2 5>, angularLimitsMin=<0 -60 -5>, angularTarget=<0 0 0>}
      localJointOffset=<0 -1 1>
     }]
    relativeSpaceBone {boneName=root}
   }
   Output Pose(Result=Component To Local(ComponentPose=@n7)) {layerGroup=None}
  ]}
 EventGraph {nodes [
   Event BlueprintUpdateAnimation()
   n1 = Get bIsPlayingEmote()
   n2 = Get bIsOnGrindRail()
   > Set emoteorgrind(emoteorgrind:double="To Float (Boolean)"(InBool=OR Boolean(A=@n1, B=@n2))) -> (Output_Get:double)
   Event OnCharacterCustomizationLoaded()
   > Set BodyMesh(BodyMesh=GetOwnerSkeletalMeshComponent())
   > n3 = Cast To PlayerPawn_Athena(Object=TryGetPawnOwner())
   > n4 = Set PawnOwner(PawnOwner:Player Pawn Athena Object Reference=@n3) -> (Output_Get:Player Pawn Athena Object Reference)
   > Set TailMesh(TailMesh:Skeletal Mesh Component Object Reference=GetSkeletalMeshForPartType(self=@n4, PartType=MiscOrTail)) -> (Output_Get:Skeletal Mesh Component Object Reference)
   n5 = CheckForBackbling()
   > Set isWearingBackbling(isWearingBackbling=@n5)
   > Set isWearingCape(isWearingCape=@n5.bIsBackblingLong)
   > Set characterHasTail(characterHasTail:bool=CheckForTail!(self:Custom Character Part Anim Instance Object Reference) -> (bHasTail:bool)) -> (Output_Get:bool)
   > Branch(Condition=Get characterHasTail() -> (characterHasTail:bool))
   > n6 = Branch(Condition=Get isWearingCape() -> (isWearingCape:bool))
   n7 = Get TailMesh() -> (TailMesh:Skeletal Mesh Component Object Reference)
   > SetVisibility!(execute, self:Scene Component Object Reference=@n7, bNewVisibility:bool=false, bPropagateToChildren:bool=false) -> (then)
   SetVisibility(execute=@n6.else, self=@n7, bNewVisibility=true, bPropagateToChildren=false)
  ]}
]
variables [
 BodyMesh {type=SkeletalMeshComponent}
 characterHasTail {type=boolean}
 emoteorgrind {type=number}
 isWearingBackbling {type=boolean}
 isWearingCape {type=boolean}
 PawnOwner {type=PlayerPawn_Athena_C}
 TailMesh {type=SkeletalMeshComponent}
]
