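# Control Rig for the face accessory part of the M_MED_SliceVine_LoopPlum character that adds an occasional procedural twitch to the left cat ear.
# A seeded random value, re-rolled on the same interval as a looping timer, gates the motion so most intervals stay still.
# Otherwise, a short spike curve over the timer drives Engage, which weights a rotation on dyn_catEar_l_01.
# The AlphaRandomizer variable is unused.
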ControlRigBlueprint /BRCosmetics/Characters/Player/Male/Medium/Bodies/M_MED_SliceVine_LoopPlum/Meshes/Parts/M_MED_SliceVine_LoopPlum_FaceAcc_CtrlRig
graphs [
 RigVMFunctionLibrary {links [], nodes []}
 RigVMModel {nodes [
   RigUnit_BeginExecution = &RigUnit_BeginExecution()
   LessEqual = /Script/RigVM.RigVMFunction_MathDoubleLessEqual(
    B=0
    A=/Script/RigVM.RigVMFunction_RandomFloat(Duration=1.5, Maximum=1, Minimum=-3, Seed=217) {is=struct RigVMUnitNode, template="Random::Execute(in Seed,in Minimum,in Maximum,in Duration,out Result)"}.Result
   ) {is=struct RigVMUnitNode, template="LessEqual::Execute(in A,in B,out Result)"}
   RigVMFunction_ControlFlowBranch = &RigVMFunction_ControlFlowBranch(Condition=@LessEqual.Result, ExecuteContext=@RigUnit_BeginExecution.ExecutePin)
   &RigUnit_ModifyTransforms(
    ItemToModify=((Item=(Type=Bone,Name="dyn_catEar_l_01"),Transform=(Rotation=(X=-0.2151,Y=-0.0809,Z=-0.2377,W=0.9438),Translation=(X=0,Y=0,Z=0),Scale3D=(X=1,Y=1,Z=1))))
    WeightMinimum=0
    ExecutePin=@RigVMFunction_ControlFlowBranch.Completed
    Weight=RigVMVariableNode() {variable=Engage}.Value
   )
   RigVMFunction_AnimEvalRichCurve = &RigVMFunction_AnimEvalRichCurve(
    Curve=(EditorCurveData=(Keys=((Time=0.05,Value=0.01),(Time=0.15,Value=1.01),(Time=0.25,Value=0.01))))
    SourceMinimum=0
    TargetMinimum=0
    Value=/Script/RigVM.RigVMFunction_TimeLoop(Duration=1.5, Normalize=true, Speed=1) {is=struct RigVMUnitNode}.Relative
   )
   RigVMVariableNode(Value=@RigVMFunction_AnimEvalRichCurve.Result, ExecuteContext=@RigVMFunction_ControlFlowBranch.False) {variable=Engage}
   RigVMVariableNode(ExecuteContext=@RigVMFunction_ControlFlowBranch.True) {variable=Engage}
  ]}
]
hierarchy [
 root {loc=<0 0 0>, rot=<0 0 0 1>}
 pelvis {parent=root, loc=<0 -0.7071 93.1191>, rot=<0.0127 0.707 -0.0127 -0.707>}
 spine_01 {parent=pelvis, loc=<6.7494 -0.6388 0>, rot=<0 0 0.034 -0.9994>}
 spine_02 {parent=spine_01, loc=<7.252 -0.035 0>, rot=<0 0 0.0238 -0.9997>}
 spine_03 {parent=spine_02, loc=<6.0106 0 0>, rot=<0 0 0.0149 -0.9999>}
 spine_04 {parent=spine_03, loc=<8.1023 0 0>, rot=<0 0 -0.0132 -0.9999>}
 spine_05 {parent=spine_04, loc=<10.0142 0 0>, rot=<0 0 -0.0015 -1>}
 neck_01 {parent=spine_05, loc=<11.3138 0.1529 0>, rot=<0 0 -0.2003 -0.9797>}
 neck_02 {parent=neck_01, loc=<4.5627 0 0>, rot=<0 0 0.0504 -0.9987>}
 head {parent=neck_02, loc=<5.3914 0 0>, rot=<0 0 0.115 -0.9934>}
 dyn_catEar_l_01 {parent=head, loc=<12.9698 -4.1613 -6.9982>, rot=<-0.7162 -0.2026 0.1765 -0.6441>}
 dyn_catEar_l_02 {parent=dyn_catEar_l_01, loc=<5.7119 0 0>, rot=<0 0 -0.0436 -0.999>}
 dyn_catEar_r_01 {parent=head, loc=<11.1568 -3.8112 7.8857>, rot=<-0.627 0.2894 -0.4535 -0.5634>}
 dyn_catEar_r_02 {parent=dyn_catEar_r_01, loc=<4.7647 0 0>, rot=<0.0871 0.0038 0.0435 -0.9952>}
 faceAttach {parent=head, loc=<-4.146 -0.8545 0>, rot=<0.0029 -0.7071 -0.0029 -0.7071>}
 clavicle_l {parent=spine_05, loc=<4.3575 -0.1466 -2.3223>, rot=<0.0465 -0.7289 -0.009 -0.6829>}
 upperarm_l {parent=clavicle_l, loc=<15.5111 0 0>, rot=<0.0226 -0.3234 -0.0088 -0.9459>}
 lowerarm_l {parent=upperarm_l, loc=<26.6514 0 0>, rot=<0.0391 -0.0099 -0.2453 -0.9686>}
 hand_l {parent=lowerarm_l, loc=<22.3491 0 0>, rot=<-0.7081 -0.0507 0.0975 -0.6975>}
 lowerarm_twist_02_l {parent=lowerarm_l, loc=<9.6199 0 0>, rot=<0 0 0 -1>}
 lowerarm_twist_01_l {parent=lowerarm_l, loc=<17.8549 0 0>, rot=<0 0 0 -1>}
 upperarm_twist_01_l {parent=upperarm_l, loc=<0.0041 0 0>, rot=<0 0 0 -1>}
 upperarm_twist_02_l {parent=upperarm_l, loc=<13.7941 0 0>, rot=<0 0 0 -1>}
 clavicle_r {parent=spine_05, loc=<4.3575 -0.1466 2.3223>, rot=<0.7289 0.0465 -0.6829 0.009>}
 upperarm_r {parent=clavicle_r, loc=<-15.5111 0 0>, rot=<0.0226 -0.3234 -0.0088 -0.9459>}
 lowerarm_r {parent=upperarm_r, loc=<-26.6514 0 0>, rot=<0.0391 -0.0099 -0.2453 -0.9686>}
 hand_r {parent=lowerarm_r, loc=<-22.3491 0 0>, rot=<-0.7081 -0.0507 0.0975 -0.6975>}
 lowerarm_twist_02_r {parent=lowerarm_r, loc=<-9.6199 0 0>, rot=<0 0 0 -1>}
 lowerarm_twist_01_r {parent=lowerarm_r, loc=<-17.8549 0 0>, rot=<0 0 0 -1>}
 upperarm_twist_01_r {parent=upperarm_r, loc=<-0.0041 0 0>, rot=<0 0 0 -1>}
 upperarm_twist_02_r {parent=upperarm_r, loc=<-13.7941 0 0>, rot=<0 0 0 -1>}
 dyn_hoodieString_l_base {parent=spine_05, loc=<10.9918 9.0272 -3.3535>, rot=<-0.2232 -0.7004 -0.6281 -0.2552>}
 dyn_hoodieString_l_01 {parent=dyn_hoodieString_l_base, loc=<0 0 0>, rot=<-0.0004 0.0002 0.001 -1>}
 dyn_hoodieString_l_02 {parent=dyn_hoodieString_l_01, loc=<8.4085 0 -0.7356>, rot=<-0.0042 -0.1266 0.0032 -0.9919>}
 dyn_hoodieString_l_03 {parent=dyn_hoodieString_l_02, loc=<7.9889 0 -0.6989>, rot=<-0.0355 -0.1147 0.0151 -0.9927>}
 dyn_hoodieString_l_04 {parent=dyn_hoodieString_l_03, loc=<8.7201 0 -0.7629>, rot=<0.0338 -0.1229 0.002 -0.9918>}
 dyn_hoodieString_l_05 {parent=dyn_hoodieString_l_04, loc=<8.514 0 0>, rot=<0 0 0 1>}
 dyn_hoodieString_r_base {parent=spine_05, loc=<9.0557 9.1295 2.6429>, rot=<0.3025 0.6069 0.6878 0.2592>}
 dyn_hoodieString_r_01 {parent=dyn_hoodieString_r_base, loc=<0 0 0>, rot=<0.0004 -0.0002 -0.0002 -1>}
 dyn_hoodieString_r_02 {parent=dyn_hoodieString_r_01, loc=<7.0955 0 -0.6208>, rot=<-0.0033 -0.1752 -0.0127 -0.9844>}
 dyn_hoodieString_r_03 {parent=dyn_hoodieString_r_02, loc=<7.826 -0.1371 -0.6847>, rot=<-0.0138 -0.1612 0.014 -0.9867>}
 dyn_hoodieString_r_04 {parent=dyn_hoodieString_r_03, loc=<8.7259 0 0>, rot=<-0.0126 -0.0616 0.0068 -0.998>}
 dyn_hoodieString_r_05 {parent=dyn_hoodieString_r_04, loc=<7.4731 0 0>, rot=<0.0262 0 0 -0.9997>}
 thigh_l {parent=pelvis, loc=<-2.5693 -0.0109 -7.7501>, rot=<0.0012 0.0572 0.0155 -0.9982>}
 calf_l {parent=thigh_l, loc=<-39.0484 0 0>, rot=<0.0006 -0.0161 -0.0391 -0.9991>}
 foot_l {parent=calf_l, loc=<-42.561 0.1077 0.4292>, rot=<0.0418 -0.0455 0.0343 -0.9975>}
 ball_l {parent=foot_l, loc=<-6.1599 13.8264 0.6806>, rot=<0.0001 -0.0001 -0.7063 -0.7079>}
 calf_twist_02_l {parent=calf_l, loc=<-18.8064 -0.0967 0.0397>, rot=<0.0237 0.0095 -0.0035 -0.9997>}
 calf_twist_01_l {parent=calf_l, loc=<-35.813 -0.0967 0.0397>, rot=<0.0237 0.0095 -0.0035 -0.9997>}
 thigh_twist_01_l {parent=thigh_l, loc=<-21.6303 0 0>, rot=<0.0013 0 0 -1>}
 thigh_r {parent=pelvis, loc=<-2.5693 -0.0109 7.7501>, rot=<0.0572 -0.0012 0.9982 0.0155>}
 calf_r {parent=thigh_r, loc=<39.0484 0 0>, rot=<0.0006 -0.0161 -0.0391 -0.9991>}
 foot_r {parent=calf_r, loc=<42.561 -0.1077 -0.4292>, rot=<0.0418 -0.0455 0.0343 -0.9975>}
 ball_r {parent=foot_r, loc=<6.1599 -13.8264 -0.6806>, rot=<0.0001 -0.0001 -0.7063 -0.7079>}
 calf_twist_02_r {parent=calf_r, loc=<18.8064 0.0967 -0.0397>, rot=<0.0237 0.0095 -0.0035 -0.9997>}
 calf_twist_01_r {parent=calf_r, loc=<35.813 0.0967 -0.0397>, rot=<0.0237 0.0095 -0.0035 -0.9997>}
 thigh_twist_01_r {parent=thigh_r, loc=<21.6303 0 0>, rot=<0.0013 0 0 -1>}
]
variables [
 AlphaRandomizer {type=double, default=0}
 Engage {type=double, default=0}
]
