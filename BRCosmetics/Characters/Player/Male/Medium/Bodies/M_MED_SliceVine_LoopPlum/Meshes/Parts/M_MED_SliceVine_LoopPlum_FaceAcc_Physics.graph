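# Physics asset for the FaceAcc part of the SliceVine LoopPlum male medium body, covering the dynamic hoodie drawstrings and the upper body around them.
# Two six-bone capsule chains, left and right, hang from spine_05, with swing limits along each string and zero twist limit below each base.
# Larger capsules on the upper spine, clavicles and upper arms, joined by free angular constraints, approximate the torso, likely as colliders for the strings.
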PhysicsAsset /BRCosmetics/Characters/Player/Male/Medium/Bodies/M_MED_SliceVine_LoopPlum/Meshes/Parts/M_MED_SliceVine_LoopPlum_FaceAcc_Physics
bodies {
 clavicle_l [{shapeType=Capsule, length=11, radius=8.5, shapeName=clavicle_l_capsule, center=<8.0382 0 0>, rotation=<rot -90 90 -89.9998>}]
 clavicle_r [{shapeType=Capsule, length=11, radius=8.5, shapeName=clavicle_r_capsule, center=<-8.0382 -0.0001 0>, rotation=<rot 90 -179.9899 -179.9903>}]
 dyn_hoodieString_l_01 [{shapeType=Capsule, length=1.5, radius=1.5, shapeName=dyn_hoodieString_l_01_capsule, center=<2.6424 0 0>, rotation=<rot -90 180 -180>}]
 dyn_hoodieString_l_02 [{shapeType=Capsule, length=1, radius=2, shapeName=dyn_hoodieString_l_02_capsule, center=<2.6424 0 0>, rotation=<rot -90 179.9994 -179.9994>}]
 dyn_hoodieString_l_03 [{shapeType=Capsule, length=1, radius=2, shapeName=dyn_hoodieString_l_03_capsule, center=<2.6424 0 0>, rotation=<rot -90 180 -179.9999>}]
 dyn_hoodieString_l_04 [{shapeType=Capsule, length=1, radius=2, shapeName=dyn_hoodieString_l_04_capsule, center=<2.6424 0 0>, rotation=<rot -90 180 -179.9999>}]
 dyn_hoodieString_l_05 [{shapeType=Capsule, length=1, radius=2, shapeName=dyn_hoodieString_l_05_capsule, center=<2.6424 0 0>, rotation=<rot -90 -180 -180>}]
 dyn_hoodieString_l_base [{shapeType=Capsule, length=4.5, radius=1, shapeName=dyn_hoodieString_l_base_capsule, center=<4.3894 0 0>, rotation=<rot -90 0.0099 -0.0099>}]
 dyn_hoodieString_r_01 [{
   shapeType=Capsule
   length=2.5
   radius=1.5
   shapeName=dyn_hoodieString_r_01_capsule
   center=<2.6424 0 0>
   rotation=<rot -90 -179.9604 179.9604>
  }]
 dyn_hoodieString_r_02 [{shapeType=Capsule, length=1.5, radius=1.5, shapeName=dyn_hoodieString_r_02_capsule, center=<2.6424 0 0>, rotation=<rot -90 180 180>}]
 dyn_hoodieString_r_03 [{shapeType=Capsule, length=1, radius=2, shapeName=dyn_hoodieString_r_03_capsule, center=<2.6424 0 0>, rotation=<rot -90 0 0>}]
 dyn_hoodieString_r_04 [{shapeType=Capsule, length=1, radius=2, shapeName=dyn_hoodieString_r_04_capsule, center=<2.6424 0 0>, rotation=<rot -90 -0.0396 0.0396>}]
 dyn_hoodieString_r_05 [{shapeType=Capsule, length=1, radius=2, shapeName=dyn_hoodieString_r_05_capsule, center=<2.1995 0 0>, rotation=<rot -90 180 180>}]
 dyn_hoodieString_r_base [{shapeType=Capsule, length=4.5, radius=1, shapeName=dyn_hoodieString_r_base_capsule, center=<4.3894 0 0>, rotation=<rot -90 -90 90>}]
 spine_03 [{shapeType=Capsule, length=10, radius=16, shapeName=spine_03_capsule, center=<0 0.1071 0>}]
 spine_04 [{shapeType=Capsule, length=9, radius=15, shapeName=spine_04_capsule, center=<1.6296 0.814 0>}]
 spine_05 [{shapeType=Capsule, length=14.5, radius=12, shapeName=spine_05_capsule, center=<-1.2999 0.673 0>}]
 upperarm_l [{shapeType=Capsule, length=24.5, radius=9, shapeName=upperarm_l_capsule, center=<14.4121 0.0001 0>, rotation=<rot 90 -179.9602 -179.9607>}]
 upperarm_r [{shapeType=Capsule, length=24.5, radius=9, shapeName=upperarm_r_capsule, center=<-14.4121 0 0>, rotation=<rot -90 179.995 -179.9951>}]
}
constraints [
 {
  bone1Name=upperarm_l
  bone2Name=clavicle_l
  swing1LimitDegrees=45
  swing1Motion=Free
  swing2LimitDegrees=45
  swing2Motion=Free
  twistLimitDegrees=45
  twistMotion=Free
 }
 {
  bone1Name=clavicle_l
  bone2Name=spine_05
  swing1LimitDegrees=45
  swing1Motion=Free
  swing2LimitDegrees=45
  swing2Motion=Free
  twistLimitDegrees=45
  twistMotion=Free
 }
 {
  bone1Name=spine_05
  bone2Name=spine_04
  swing1LimitDegrees=45
  swing1Motion=Free
  swing2LimitDegrees=45
  swing2Motion=Free
  twistLimitDegrees=45
  twistMotion=Free
 }
 {
  bone1Name=spine_04
  bone2Name=spine_03
  swing1LimitDegrees=45
  swing1Motion=Free
  swing2LimitDegrees=45
  swing2Motion=Free
  twistLimitDegrees=45
  twistMotion=Free
 }
 {
  bone1Name=clavicle_r
  bone2Name=spine_05
  swing1LimitDegrees=45
  swing1Motion=Free
  swing2LimitDegrees=45
  swing2Motion=Free
  twistLimitDegrees=45
  twistMotion=Free
 }
 {
  bone1Name=upperarm_r
  bone2Name=clavicle_r
  swing1LimitDegrees=45
  swing1Motion=Free
  swing2LimitDegrees=45
  swing2Motion=Free
  twistLimitDegrees=45
  twistMotion=Free
 }
 {bone1Name=dyn_hoodieString_l_base, bone2Name=spine_05, swing1LimitDegrees=45, swing2LimitDegrees=45, twistLimitDegrees=45}
 {bone1Name=dyn_hoodieString_l_01, bone2Name=dyn_hoodieString_l_base, swing1LimitDegrees=45, swing2LimitDegrees=45, twistLimitDegrees=0}
 {bone1Name=dyn_hoodieString_l_02, bone2Name=dyn_hoodieString_l_01, swing1LimitDegrees=45, swing2LimitDegrees=45, twistLimitDegrees=0}
 {bone1Name=dyn_hoodieString_l_03, bone2Name=dyn_hoodieString_l_02, swing1LimitDegrees=45, swing2LimitDegrees=45, twistLimitDegrees=0}
 {bone1Name=dyn_hoodieString_l_04, bone2Name=dyn_hoodieString_l_03, swing1LimitDegrees=45, swing2LimitDegrees=45, twistLimitDegrees=0}
 {bone1Name=dyn_hoodieString_l_05, bone2Name=dyn_hoodieString_l_04, swing1LimitDegrees=45, swing2LimitDegrees=45, twistLimitDegrees=0}
 {bone1Name=dyn_hoodieString_r_base, bone2Name=spine_05, swing1LimitDegrees=45, swing2LimitDegrees=45, twistLimitDegrees=45}
 {bone1Name=dyn_hoodieString_r_01, bone2Name=dyn_hoodieString_r_base, swing1LimitDegrees=45, swing2LimitDegrees=45, twistLimitDegrees=0}
 {bone1Name=dyn_hoodieString_r_02, bone2Name=dyn_hoodieString_r_01, swing1LimitDegrees=45, swing2LimitDegrees=45, twistLimitDegrees=0}
 {bone1Name=dyn_hoodieString_r_03, bone2Name=dyn_hoodieString_r_02, swing1LimitDegrees=45, swing2LimitDegrees=45, twistLimitDegrees=0}
 {bone1Name=dyn_hoodieString_r_04, bone2Name=dyn_hoodieString_r_03, swing1LimitDegrees=45, swing2LimitDegrees=45, twistLimitDegrees=0}
 {bone1Name=dyn_hoodieString_r_05, bone2Name=dyn_hoodieString_r_04, swing1LimitDegrees=45, swing2LimitDegrees=45, twistLimitDegrees=0}
]
