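# Post-copy animation blueprint for the LoopPlum face accessory, adding secondary motion including cat ears.
# It copies the attached parent's pose, layers the Body_Mesh pose from the clavicles down, then runs and caches a Control Rig.
# A spine-relative RigidBody simulation follows, then four chained AnimDynamics nodes spring both two-bone ears with tight angular limits.
# Body_Mesh is picked from the custom mesh parts when character customization loads.
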AnimBlueprint /BRCosmetics/Characters/Player/Male/Medium/Bodies/M_MED_SliceVine_LoopPlum/Meshes/Parts/M_MED_SliceVine_LoopPlum_FaceAcc_AnimBP
target_skeleton=&/M_MED_SliceVine_LoopPlum_FaceAcc_Skeleton
graphs [
 AnimGraph {nodes [
   n1 = RigidBody(ComponentPose=Local To Component(LocalPose=Use cached pose 'CtrlRigCache'())) {
    bClampLinearTranslationLimitToRefPose=true
    bOverrideWorldGravity=true
    overridePhysicsAsset=&/M_MED_SliceVine_LoopPlum_FaceAcc_Physics
    simulationSpace=BaseBoneSpace
    alphaBoolBlend {blendOption=Linear}
    baseBoneRef {boneName=spine_05}
    componentAppliedLinearAccClamp=<3000 3000 3000>
    componentLinearAccScale=<0.3 0.3 0.3>
    componentLinearVelScale=<0.4 0.4 0.4>
    overrideWorldGravity {z=-980}
   }
   n2 = Anim Dynamics - Bone: dyn_catEar_l_01(ComponentPose=@n1, Alpha=1, ExternalForce="0,0,0") {
    angularDampingOverride=1
    angularSpringConstant=10000
    bAngularSpring=true
    bOverrideAngularDamping=true
    bUseGravityOverride=true
    gravityScale=0.35
    simulationSpace=BoneRelative
    alphaBoolBlend {blendOption=Linear}
    boundBone {boneName=dyn_catEar_l_01}
    componentAppliedLinearAccClamp=<1300 1300 1300>
    componentLinearAccScale=<1 1 1>
    componentLinearVelScale=<1 1 1>
    gravityOverride {z=-980}
    physicsBodyDefinitions [{
      boundBone {boneName=dyn_catEar_l_01}
      boxExtents=<15 20 20>
      constraintSetup {angularLimitsMax=<2 5 2>, angularLimitsMin=<-2 -5 -2>, angularTarget=<1 0 0>}
      localJointOffset=<2 -0.2 0>
     }]
    relativeSpaceBone {boneName=spine_04}
   }
   n3 = Anim Dynamics - Bone: dyn_catEar_l_02(ComponentPose=@n2, Alpha=1, ExternalForce="0,0,0") {
    angularDampingOverride=1
    angularSpringConstant=12000
    bAngularSpring=true
    bOverrideAngularDamping=true
    bUseGravityOverride=true
    gravityScale=0.35
    simulationSpace=BoneRelative
    alphaBoolBlend {blendOption=Linear}
    boundBone {boneName=dyn_catEar_l_02}
    componentAppliedLinearAccClamp=<1300 1300 1300>
    componentLinearAccScale=<1 1 1>
    componentLinearVelScale=<1 1 1>
    gravityOverride {z=-980}
    physicsBodyDefinitions [{
      boundBone {boneName=dyn_catEar_l_02}
      boxExtents=<15 20 20>
      constraintSetup {angularLimitsMax=<2 5 2>, angularLimitsMin=<-2 -5 -2>, angularTarget=<1 0 0>}
      localJointOffset=<2 -0.2 0>
     }]
    relativeSpaceBone {boneName=spine_04}
   }
   n4 = Anim Dynamics - Bone: dyn_catEar_r_01(ComponentPose=@n3, Alpha=1, ExternalForce="0,0,0") {
    angularDampingOverride=1
    angularSpringConstant=10000
    bAngularSpring=true
    bOverrideAngularDamping=true
    bUseGravityOverride=true
    gravityScale=0.35
    simulationSpace=BoneRelative
    alphaBoolBlend {blendOption=Linear}
    boundBone {boneName=dyn_catEar_r_01}
    componentAppliedLinearAccClamp=<1300 1300 1300>
    componentLinearAccScale=<1 1 1>
    componentLinearVelScale=<1 1 1>
    gravityOverride {z=-980}
    physicsBodyDefinitions [{
      boundBone {boneName=dyn_catEar_r_01}
      boxExtents=<15 20 20>
      constraintSetup {angularLimitsMax=<2 10 2>, angularLimitsMin=<-2 -10 -2>, angularTarget=<1 0 0>}
      localJointOffset=<-2 0.5 -0.5>
     }]
    relativeSpaceBone {boneName=spine_04}
   }
   n5 = Anim Dynamics - Bone: dyn_catEar_r_02(ComponentPose=@n4, Alpha=1, ExternalForce="0,0,0") {
    angularDampingOverride=1
    angularSpringConstant=12000
    bAngularSpring=true
    bOverrideAngularDamping=true
    bUseGravityOverride=true
    gravityScale=0.35
    simulationSpace=BoneRelative
    alphaBoolBlend {blendOption=Linear}
    boundBone {boneName=dyn_catEar_r_02}
    componentAppliedLinearAccClamp=<1300 1300 1300>
    componentLinearAccScale=<1 1 1>
    componentLinearVelScale=<1 1 1>
    gravityOverride {z=-980}
    physicsBodyDefinitions [{
      boundBone {boneName=dyn_catEar_r_02}
      boxExtents=<15 20 20>
      constraintSetup {angularLimitsMax=<2 10 2>, angularLimitsMin=<-2 -10 -2>, angularTarget=<1 0 0>}
      localJointOffset=<-2 0.5 -0.5>
     }]
    relativeSpaceBone {boneName=spine_04}
   }
   Output Pose(Result=Component To Local(ComponentPose=@n5)) {layerGroup=None}
   n6 = Get Body_Mesh() -> (Body_Mesh:Skeletal Mesh Component Budgeted Object Reference)
   n7 = Control Rig(Source=Layered blend per bone(
     BasePose=Copy Pose From Mesh() {bUseAttachedParent=true}
     BlendPoses_0=Copy Pose From Mesh(SourceMeshComponent=@n6) {bUseAttachedParent=true}
    ) {layerSetup [{branchFilters [{boneName=clavicle_l}, {boneName=clavicle_r}]}]}) {
    alphaBoolBlend {blendOption=Linear}
    controlRigAssetReference {blueprintRigClass=&/M_MED_SliceVine_LoopPlum_FaceAcc_CtrlRig.M_MED_SliceVine_LoopPlum_FaceAcc_CtrlRig_C}
   }
   Save cached pose 'CtrlRigCache'(Pose=@n7) {cacheName=CtrlRigCache}
  ]}
 EventGraph {nodes [
   n1 = Event OnCharacterCustomizationLoaded()
   > Set Body_Mesh(Body_Mesh:Skeletal Mesh Component Budgeted Object Reference=Get Array Item with EFort Custom Part Type(Array=@n1.CustomMeshParts)) -> (Output_Get:Skeletal Mesh Component Budgeted Object Reference)
   Event BlueprintUpdateAnimation()
  ]}
]
variables [Body_Mesh {type=SkeletalMeshComponentBudgeted}]
